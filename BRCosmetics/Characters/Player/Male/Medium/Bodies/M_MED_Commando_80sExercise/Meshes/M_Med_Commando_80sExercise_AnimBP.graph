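# Animation Blueprint for the male 80s Exercise Commando body mesh, copying a source mesh's pose and adding secondary motion to tags and collar bones.
# On initialization it stores the result of GetOwnerSkeletalMeshComponent as MeshToCopy, which feeds Copy Pose From Mesh.
# Three chained Anim Dynamics nodes then simulate dyn_tags, dyn_collar_l and dyn_collar_r in actor space with angular limits.
# Angular springs are enabled on the tags and left collar, but not the right collar.
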AnimBlueprint /BRCosmetics/Characters/Player/Male/Medium/Bodies/M_MED_Commando_80sExercise/Meshes/M_Med_Commando_80sExercise_AnimBP
target_skeleton=&/M_Med_Commando_80sExercise_Skeleton
graphs [
 AnimGraph {nodes [
   n1 = Get MeshToCopy() -> (MeshToCopy:Skeletal Mesh Component Object Reference)
   n2 = Anim Dynamics - Bone: dyn_tags(ComponentPose=Local To Component(LocalPose=Copy Pose From Mesh(SourceMeshComponent=@n1)), Alpha=1, ExternalForce="0,0,0") {
    alpha=0.9
    angularSpringConstant=1500
    bAngularSpring=true
    bGravityOverrideInSimSpace=true
    numSolverIterationsPostUpdate=4
    numSolverIterationsPreUpdate=16
    simulationSpace=Actor
    alphaBoolBlend {blendOption=Linear}
    boundBone {boneName=dyn_tags}
    physicsBodyDefinitions [{
      sphereCollisionRadius=0
      boundBone {boneName=dyn_tags}
      boxExtents=<15 5 5>
      constraintSetup {angularTargetAxis=AxisZ, twistAxis=AxisY, angularLimitsMax=<15 0 15>, angularLimitsMin=<-15 0 -15>, angularTarget=<0 0 1>}
      localJointOffset=<0 0 0>
     }]
   }
   n3 = Anim Dynamics - Bone: dyn_collar_l(ComponentPose=@n2, Alpha=1, ExternalForce="0,0,0") {
    angularSpringConstant=10
    bAngularSpring=true
    bGravityOverrideInSimSpace=true
    gravityScale=0.5
    simulationSpace=Actor
    alphaBoolBlend {blendOption=Linear}
    boundBone {boneName=dyn_collar_l}
    componentLinearAccScale=<1 1 1>
    physicsBodyDefinitions [{
      sphereCollisionRadius=0
      boundBone {boneName=dyn_collar_l}
      boxExtents=<10 5 10>
      constraintSetup {angularLimitsMax=<20 20 20>, angularLimitsMin=<-20 -20 -20>, angularTarget=<0 0 0>}
      localJointOffset=<0 0.2 -0.1>
     }]
   }
   n4 = Anim Dynamics - Bone: dyn_collar_r(ComponentPose=@n3, Alpha=1, ExternalForce="0,0,0") {
    angularSpringConstant=10
    bGravityOverrideInSimSpace=true
    gravityScale=0.5
    simulationSpace=Actor
    alphaBoolBlend {blendOption=Linear}
    boundBone {boneName=dyn_collar_r}
    componentLinearAccScale=<1 1 1>
    physicsBodyDefinitions [{
      sphereCollisionRadius=0
      boundBone {boneName=dyn_collar_r}
      boxExtents=<10 5 7>
      constraintSetup {angularLimitsMax=<20 20 20>, angularLimitsMin=<-20 -20 -20>, angularTarget=<0 0 0>}
      localJointOffset=<0 0.2 -0.1>
     }]
   }
   Output Pose(Result=Component To Local(ComponentPose=@n4)) {layerGroup=None}
  ]}
 EventGraph {nodes [
   Event BlueprintInitializeAnimation()
   > Set MeshToCopy(MeshToCopy:Skeletal Mesh Component Object Reference=GetOwnerSkeletalMeshComponent()) -> (Output_Get:Skeletal Mesh Component Object Reference)
   TryGetPawnOwner()
   Event BlueprintUpdateAnimation()
  ]}
]
variables [MeshToCopy {type=SkeletalMeshComponent}]
